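# Animation Blueprint for the GoggleBud glider cosmetic's skeleton, letting montages animate the glider mesh.
# The AnimGraph passes its input pose through the DefaultSlot montage slot straight to the output pose, with no state machine or variables.
# The EventGraph holds only unconnected update event, pawn owner, self-class cast and Play Montage nodes, so it runs no logic of its own.
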AnimBlueprint /BRCosmetics/Gadgets/Assets/VinderTech_GliderChute/Glider_GoggleBud/Meshes/GoggleBud_Glider_AnimBP
target_skeleton=&/GoggleBud_Glider_Skeleton
graphs [
 AnimGraph {nodes [
   AnimGraphNode_SubInput_0 = Input Pose()
   Output Pose(Result=Slot 'DefaultSlot'(Source=@AnimGraphNode_SubInput_0)) {layerGroup=None}
  ]}
 EventGraph {nodes [
   TryGetPawnOwner()
   Cast To GoggleBud_Glider_AnimBP!(execute, Object:Wildcard) -> (then, CastFailed, AsGoggle Bud Glider Anim BP:Goggle Bud Glider Anim BP Object Reference) {is=title K2Node_DynamicCast}
   Event BlueprintUpdateAnimation()
   Play Montage!(
    execute:Exec
    InSkeletalMeshComponent:Skeletal Mesh Component Object Reference
    MontageToPlay:Anim Montage Object Reference
    PlayRate:float=1
    StartingPosition:float=0
    StartingSection:Name=None
    bShouldStopAllMontages:bool=true
   ) -> (then:Exec, OnCompleted:Exec, OnBlendOut:Exec, OnInterrupted:Exec, OnNotifyBegin:Exec, OnNotifyEnd:Exec, NotifyName:Name) {is=title K2Node_PlayMontage}
  ]}
]
variables []
